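; Parent time-of-day manager Blueprint for Figment_S03 that sets sky, lighting and storm visuals.
; Construction and BeginPlay push skybox fog settings into material parameter collections; BeginPlay also publishes shadow settings and server-spawns comet event components if absent.
; A storm timeline drives strength, rain and foliage wind, while self-rescheduling lightning flashes the directional light and counter-scales fog inscattering.
; Heavy SSAO substitutes when distance-field AO is off, refreshed on user-settings updates.
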(event EventBeginPlay
  (bind _directionallightcomponent (Variables|DayNightCycle|GetDirectionalLightComponent))
  (Utilities|FlowControl|Sequence
    (:then_0
      ;[1 1 1] destroy hacky post process at run time
      (Components|DestroyComponent (Variables|Default|GetEditorPostProcessMaterial(HackToWorkAroundTheOutlineShaderBeingInjectedByCodeAtRuntime))))
    (:then_1
      ;[1 1 1] Skybox fake fog material settings
      (Rendering|Material|SetVectorParameterValue "/Game/Environments/World/Backgrounds/Universal/MasterMaterials/MPC_Skybox_World.MPC_Skybox_World" "MorningColor" (Variables|SkyboxFogSettings|GetSkyboxFogMorningColor) :!class /Script/Engine.KismetMaterialLibrary)
      (Rendering|Material|SetVectorParameterValue "/Game/Environments/World/Backgrounds/Universal/MasterMaterials/MPC_Skybox_World.MPC_Skybox_World" "MorningLitColor" (Variables|SkyboxFogSettings|GetSkyboxFogMorningLitColor) :!class /Script/Engine.KismetMaterialLibrary)
      (Rendering|Material|SetVectorParameterValue "/Game/Environments/World/Backgrounds/Universal/MasterMaterials/MPC_Skybox_World.MPC_Skybox_World" "Morn_R=Distance_G=DiffuseAmt_B=EmissiveAmt" (Math|Color|MakeColor (Variables|SkyboxFogSettings|GetSkyboxFogMorningDistance) (Variables|SkyboxFogSettings|GetSkyboxFogMorningDiffuseAmount) (Variables|SkyboxFogSettings|GetSkyboxFogMorningEmissiveAmount)) :!class /Script/Engine.KismetMaterialLibrary)
      (Rendering|Material|SetVectorParameterValue "/Game/Environments/World/Backgrounds/Universal/MasterMaterials/MPC_Skybox_World.MPC_Skybox_World" "Morn_R=LitLength_G=LitRot_B=LitOffset" (Math|Color|MakeColor (Variables|SkyboxFogSettings|GetSkyboxFogMorningLitLength) (+ 4000000 (Math|Float|Clamp(Float) (Variables|SkyboxFogSettings|GetSkyboxFogMorningLitRotation) -4000000 4000000)) (+ 4000000 (Math|Float|Clamp(Float) (Variables|SkyboxFogSettings|GetSkyboxFogMorningLitOffset) -4000000 4000000))) :!class /Script/Engine.KismetMaterialLibrary)
      (Rendering|Material|SetVectorParameterValue "/Game/Environments/World/Backgrounds/Universal/MasterMaterials/MPC_Skybox_World.MPC_Skybox_World" "DayColor" (Variables|SkyboxFogSettings|GetSkyboxFogDayColor) :!class /Script/Engine.KismetMaterialLibrary)
      (Rendering|Material|SetVectorParameterValue "/Game/Environments/World/Backgrounds/Universal/MasterMaterials/MPC_Skybox_World.MPC_Skybox_World" "DayLitColor" (Variables|SkyboxFogSettings|GetSkyboxFogDayLitColor) :!class /Script/Engine.KismetMaterialLibrary)
      (Rendering|Material|SetVectorParameterValue "/Game/Environments/World/Backgrounds/Universal/MasterMaterials/MPC_Skybox_World.MPC_Skybox_World" "Day_R=Distance_G=DiffuseAmt_B=EmissiveAmt" (Math|Color|MakeColor (Variables|SkyboxFogSettings|GetSkyboxFogDayDistance) (Variables|SkyboxFogSettings|GetSkyboxFogDayDiffuseAmount) (Variables|SkyboxFogSettings|GetSkyboxFogDayEmissiveAmount)) :!class /Script/Engine.KismetMaterialLibrary)
      (Rendering|Material|SetVectorParameterValue "/Game/Environments/World/Backgrounds/Universal/MasterMaterials/MPC_Skybox_World.MPC_Skybox_World" "Day_R=LitLength_G=LitRot_B=LitOffset" (Math|Color|MakeColor (Variables|SkyboxFogSettings|GetSkyboxFogDayLitLength) (+ 4000000 (Math|Float|Clamp(Float) (Variables|SkyboxFogSettings|GetSkyboxFogDayLitRotation) -4000000 4000000)) (+ 4000000 (Math|Float|Clamp(Float) (Variables|SkyboxFogSettings|GetSkyboxFogDayLitOffset) -4000000 4000000))) :!class /Script/Engine.KismetMaterialLibrary)
      (Rendering|Material|SetVectorParameterValue "/Game/Environments/World/Backgrounds/Universal/MasterMaterials/MPC_Skybox_World.MPC_Skybox_World" "EveningColor" (Variables|SkyboxFogSettings|GetSkyboxFogEveningColor) :!class /Script/Engine.KismetMaterialLibrary)
      (Rendering|Material|SetVectorParameterValue "/Game/Environments/World/Backgrounds/Universal/MasterMaterials/MPC_Skybox_World.MPC_Skybox_World" "EveningLitColor" (Variables|SkyboxFogSettings|GetSkyboxFogEveningLitColor) :!class /Script/Engine.KismetMaterialLibrary)
      (Rendering|Material|SetVectorParameterValue "/Game/Environments/World/Backgrounds/Universal/MasterMaterials/MPC_Skybox_World.MPC_Skybox_World" "Eve_R=Distance_G=DiffuseAmt_B=EmissiveAmt" (Math|Color|MakeColor (Variables|SkyboxFogSettings|GetSkyboxFogEveningDistance) (Variables|SkyboxFogSettings|GetSkyboxFogEveningDiffuseAmount) (Variables|SkyboxFogSettings|GetSkyboxFogEveningEmissiveAmount)) :!class /Script/Engine.KismetMaterialLibrary)
      (Rendering|Material|SetVectorParameterValue "/Game/Environments/World/Backgrounds/Universal/MasterMaterials/MPC_Skybox_World.MPC_Skybox_World" "Eve_R=LitLength_G=LitRot_B=LitOffset" (Math|Color|MakeColor (Variables|SkyboxFogSettings|GetSkyboxFogEveningLitLength) (+ 4000000 (Math|Float|Clamp(Float) (Variables|SkyboxFogSettings|GetSkyboxFogEveningLitRotation) -4000000 4000000)) (+ 4000000 (Math|Float|Clamp(Float) (Variables|SkyboxFogSettings|GetSkyboxFogEveningLitOffset) -4000000 4000000))) :!class /Script/Engine.KismetMaterialLibrary)
      (Rendering|Material|SetVectorParameterValue "/Game/Environments/World/Backgrounds/Universal/MasterMaterials/MPC_Skybox_World.MPC_Skybox_World" "NightColor" (Variables|SkyboxFogSettings|GetSkyboxFogNightColor) :!class /Script/Engine.KismetMaterialLibrary)
      (Rendering|Material|SetVectorParameterValue "/Game/Environments/World/Backgrounds/Universal/MasterMaterials/MPC_Skybox_World.MPC_Skybox_World" "NightLitColor" (Variables|SkyboxFogSettings|GetSkyboxFogNightLitColor) :!class /Script/Engine.KismetMaterialLibrary)
      (Rendering|Material|SetVectorParameterValue "/Game/Environments/World/Backgrounds/Universal/MasterMaterials/MPC_Skybox_World.MPC_Skybox_World" "Night_R=Distance_G=DiffuseAmt_B=EmissiveAmt" (Math|Color|MakeColor (Variables|SkyboxFogSettings|GetSkyboxFogNightDistance) (Variables|SkyboxFogSettings|GetSkyboxFogNightDiffuseAmount) (Variables|SkyboxFogSettings|GetSkyboxFogNightEmissiveAmount)) :!class /Script/Engine.KismetMaterialLibrary)
      (Rendering|Material|SetVectorParameterValue "/Game/Environments/World/Backgrounds/Universal/MasterMaterials/MPC_Skybox_World.MPC_Skybox_World" "Night_R=LitLength_G=LitRot_B=LitOffset" (Math|Color|MakeColor (Variables|SkyboxFogSettings|GetSkyboxFogNightLitLength) (+ 4000000 (Math|Float|Clamp(Float) (Variables|SkyboxFogSettings|GetSkyboxFogNightLitRotation) -4000000 4000000)) (+ 4000000 (Math|Float|Clamp(Float) (Variables|SkyboxFogSettings|GetSkyboxFogNightLitOffset) -4000000 4000000))) :!class /Script/Engine.KismetMaterialLibrary))
    (:then_2
      ;[1 1 1] FrontEnd settings
      ;;[1 1 1] Store render settings and make them accessible to shaders. Will need to implement update function when it becomes available.
      (bind _distance_fields_enabled (Utilities|ShadowQualityCheckVFX)
        (:Low
          _)
        (:Medium
          _)
        (:High
          _)
        (:Epic
          _)
        (:Cinematic
          _))
      (Rendering|Material|SetScalarParameterValue "/Game/Packages/Fortress_SharedMaterials/GlobalMaterialParameters/FortniteMaterialParameters.FortniteMaterialParameters" "AreDistanceFieldsEnabled?" (Math|Conversions|ToFloat(Boolean) _distance_fields_enabled) :!class /Script/Engine.KismetMaterialLibrary)
      (Rendering|Material|SetScalarParameterValue "/Game/Packages/Fortress_SharedMaterials/GlobalMaterialParameters/FortniteMaterialParameters.FortniteMaterialParameters" "ShadowDistance" (Variables|/Script/Engine.DirectionalLightComponent|GetDynamicShadowDistanceMovableLight _directionallightcomponent) :!class /Script/Engine.KismetMaterialLibrary)
      (Rendering|Material|SetScalarParameterValue "/Game/Packages/Fortress_SharedMaterials/GlobalMaterialParameters/FortniteMaterialParameters.FortniteMaterialParameters" "ShadowDistanceFadeoutFraction" (Variables|/Script/Engine.DirectionalLightComponent|GetShadowDistanceFadeoutFraction _directionallightcomponent) :!class /Script/Engine.KismetMaterialLibrary))
    (:then_3
      ;[1 1 1] Apply Heavy SSAO when DFAO is disabled
      (bind _returnvalue (Settings|GetGameUserSettings))
      (EventDispatchers|BindEventtoOnGameUserSettingsUINeedsUpdate _returnvalue (delegate Custom|UpdateSSAO))
      (CallFunction|SetHeavySSAO))
    (:then_4
      ;[1 1 1] Add comet to map if actor doesn't exist - remove after 4.0
      (Networking|SwitchHasAuthority
        (:Authority
          (bind _outactors (Actor|GetAllActorsOfClass "/Figment_S03/Athena/Events/BP_Athena_Event_Components.BP_Athena_Event_Components_C"))
          (if (and (<= (Utilities|Array|Length _outactors) 0) (Variables|Default|GetEnableCometDynamicPlacement))
            (Game|SpawnActorfromClass "/Figment_S03/Athena/Events/BP_Athena_Event_Components.BP_Athena_Event_Components_C" (CallFunction|EmptyTransform) "Undefined" "OverrideRootScale" 0 0)))) ; Needs to spawn on server so it can replicate things properly
    )))

(event Custom|DisableLightAndFog
  (params (Enable :type Boolean))
  ;[1 1 1] FrontEnd settings
  (Rendering|SetActorHiddenInGame (not Enable)))

(event Custom|UpdateSSAO
  ;[1 1 1] Apply Heavy SSAO when DFAO is disabled
  (CallFunction|SetHeavySSAO))

(event Custom|LightningStrike
  (bind _lightningstrength (Variables|Storm|GetLightningStrength))
  (bind _returnvalue (Math|Float|Square _lightningstrength))
  (bind _previousdirectionallightcolor (Variables|Storm|GetPreviousDirectionalLightColor))
  (bind (_r _g _b) (Math|Color|BreakColor _previousdirectionallightcolor))
  (bind _lightning_flash_color (Variables|Misc|GetLightning%20Flash%20Color))
  ;[1 1 1] Lightning
  (Variables|Storm|SetPreviousDirectionalLightColor (Rendering|Components|Light|GetLightColor (Variables|DayNightCycle|GetDirectionalLightComponent)))
  (Components|Timeline|SetPlayRate (Variables|TODM_A_PARENT|GetFlashLightning) (/ 1 (Math|Random|RandomFloatinRange (Variables|Storm|GetLightning%20Flash%20Timeline%20Min) (Variables|Storm|GetLightning%20Flash%20Timeline%20Max))))
  (bind (_direction _lightning_intensity) (Timeline|FlashLightning PlayFromStart)
    (:Update
      (Storms|SetFogDirectionalInscatteringMultiplier (/ 1 (+ (+ (* (* _lightning_intensity _returnvalue) 25) 1.0) (- (Math|Float|float+float (* _r 0.3) (* _g 0.59) (* _b 0.11)) (Math|Float|float+float (* (Math|Color|BreakColor (Math|Color|Lerp(LinearColor) _previousdirectionallightcolor _lightning_flash_color _lightning_intensity)) 0.3) (* (.G (Math|Color|BreakColor (Math|Color|Lerp(LinearColor) _previousdirectionallightcolor _lightning_flash_color _lightning_intensity))) 0.59) (* (.B (Math|Color|BreakColor (Math|Color|Lerp(LinearColor) _previousdirectionallightcolor _lightning_flash_color _lightning_intensity))) 0.11))))))
      (Storms|SetDirectionalLightColorMultiplier (+ (* (* _lightning_intensity _returnvalue) 25) 1.0))
      (Rendering|Components|Light|SetLightColor (Variables|DayNightCycle|GetDirectionalLightComponent) (Math|Color|Lerp(LinearColor) _previousdirectionallightcolor _lightning_flash_color _lightning_intensity) true))
    (:Finished
      ;[1 1 1] Reset the timer using a new duration
      (bind _returnvalue_2 (Utilities|Time|SetTimerbyEvent (delegate Custom|LightningStrike) (Math|Random|RandomFloatinRange 1 4)))
      (Variables|Storm|SetLightningStrikeTimerHandle _returnvalue_2))) ; 1 sec
)

(event Custom|StartStorm
  ;[1 1 1] Storm Start/Stop
  (Storms|ToggleDebugForceStormEffects))

(event EventOnStormStart
  (bind _returnvalue (Storms|GetStormStrength))
  (bind _threatcloudstormamount (Variables|Storm|GetThreatCloudStormAmount))
  ;[1 1 1] Storm Start/Stop
  (Utilities|FlowControl|Sequence
    (:then_0
      (bind (_direction _storm_ramp_in_value _lightning_modulator _rainint _wind_strength) (Timeline|New%20Storm%20Timeline Play)
        (:Update
          (Storms|SetStormStrength (* _storm_ramp_in_value _threatcloudstormamount))
          (Rendering|Material|SetScalarParameterValue "/Game/Packages/Fortress_SharedMaterials/GlobalMaterialParameters/FortniteMaterialParameters.FortniteMaterialParameters" "StormStrength" _returnvalue :!class /Script/Engine.KismetMaterialLibrary)
          (Variables|Storm|SetLightningStrength (* _lightning_modulator _threatcloudstormamount))
          (bind _rainparticlesystemcomp (Variables|DayNightCycle|GetRainParticleSystemComp)
            (:then
              (Effects|Components|ParticleSystem|SetFloatParameter _rainparticlesystemcomp "StormIntensity" _returnvalue))
            (:else
              _))
          (if (or (Variables|Storm|GetUseRainWithStorms) (Variables|Storm|GetPlay%20a%20Particle%20Effect%20Near%20the%20players%20Feet))
            (Variables|Storm|SetInternal%20Rain/WindEffects%20Should%20be%20Active (Math|Conversions|ToBoolean(Integer) (Math|Float|Floor _rainint)))
            (if (Variables|Storm|GetGenerate%20Foliage%20Wind%20Gusts)
              (Rendering|Material|SetScalarParameterValue "/Game/Packages/Fortress_SharedMaterials/GlobalMaterialParameters/FortniteMaterialParameters.FortniteMaterialParameters" "Foliage Directional Wind Intensity" (* (Variables|Storm|GetStorm%20Wind%20Strength) _wind_strength) :!class /Script/Engine.KismetMaterialLibrary))))
        (:FireOffRain
          (if (Variables|Storm|GetUseRainWithStorms)
            (Components|Activation|SetActive (Variables|DayNightCycle|GetRainParticleSystemComp) (Variables|Storm|GetInternal%20Rain/WindEffects%20Should%20be%20Active) false)))))))

(event EventOnStormEnd
  (bind _returnvalue (Storms|GetStormStrength))
  (bind _threatcloudstormamount (Variables|Storm|GetThreatCloudStormAmount))
  ;[1 1 1] Storm Start/Stop
  (Utilities|FlowControl|Sequence
    (:then_0
      (Timeline|New%20Storm%20Timeline Reverse))
    (:then_1
      (Utilities|Time|ClearandInvalidateTimerbyHandle (Variables|Storm|GetLightningStrikeTimerHandle)))))

(event Custom|EnableHDRRendering
  (Utilities|FlowControl|FlipFlop
    (:A
      (Development|ExecuteConsoleCommand "fullscreen")
      (Development|ExecuteConsoleCommand "r.HDR.EnableHDROutput 1")
      (Development|ExecuteConsoleCommand "r.HDR.Display.ColorGamut 2")
      (Development|ExecuteConsoleCommand "r.HDR.Display.OutputDevice 5 ")
      (Development|ExecuteConsoleCommand "r.HDR.UI.Level .5 "))
    (:B
      (Development|ExecuteConsoleCommand "r.HDR.EnableHDROutput 0")
      (Development|ExecuteConsoleCommand "r.HDR.Display.ColorGamut 0")
      (Development|ExecuteConsoleCommand "r.HDR.Display.OutputDevice 0")
      (Development|ExecuteConsoleCommand "r.HDR.UI.Level 1 "))))

(fn EmptyTransform
  (return (Math|Transform|MakeTransform "0, 0, 0")))

(fn SetHeavySSAO
  (bind _heavyssao_postprocess (Variables|Default|GetHeavySSAO_Postprocess))
  (if (== (Fortnite|GetNumericCVar "r.DistanceFieldAO") 0.0)
    (Variables|/Script/Engine.PostProcessComponent|SetbEnabled true _heavyssao_postprocess)
    (else
      (Variables|/Script/Engine.PostProcessComponent|SetbEnabled false _heavyssao_postprocess))))

(fn ConstructionScript
  ;[1 1 1] World border cloud color
  (Rendering|Material|SetScalarParameterValue "/Game/Packages/Fortress_SharedMaterials/GlobalMaterialParameters/FortniteMaterialParameters.FortniteMaterialParameters" "World border cloud diffuse brightness" (Variables|SkyboxFogSettings|GetWorld%20Border%20Cloud%20Base%20Color%20Brightness) :!class /Script/Engine.KismetMaterialLibrary)
  (Rendering|Material|SetVectorParameterValue "/Game/Packages/Fortress_SharedMaterials/GlobalMaterialParameters/FortniteMaterialParameters.FortniteMaterialParameters" "World Border Cloud Sub Surface Color and Opacity" (Variables|SkyboxFogSettings|GetWorld%20Border%20Cloud%20Sub%20Surface%20Color%20and%20Opacity) :!class /Script/Engine.KismetMaterialLibrary)
  ;[1 1 1] Skybox fake fog material settings
  (Rendering|Material|SetVectorParameterValue "/Game/Environments/World/Backgrounds/Universal/MasterMaterials/MPC_Skybox_World.MPC_Skybox_World" "MorningColor" (Variables|SkyboxFogSettings|GetSkyboxFogMorningColor) :!class /Script/Engine.KismetMaterialLibrary)
  (Rendering|Material|SetVectorParameterValue "/Game/Environments/World/Backgrounds/Universal/MasterMaterials/MPC_Skybox_World.MPC_Skybox_World" "MorningLitColor" (Variables|SkyboxFogSettings|GetSkyboxFogMorningLitColor) :!class /Script/Engine.KismetMaterialLibrary)
  (Rendering|Material|SetVectorParameterValue "/Game/Environments/World/Backgrounds/Universal/MasterMaterials/MPC_Skybox_World.MPC_Skybox_World" "Morn_R=Distance_G=DiffuseAmt_B=EmissiveAmt" (Math|Color|MakeColor (Variables|SkyboxFogSettings|GetSkyboxFogMorningDistance) (Variables|SkyboxFogSettings|GetSkyboxFogMorningDiffuseAmount) (Variables|SkyboxFogSettings|GetSkyboxFogMorningEmissiveAmount)) :!class /Script/Engine.KismetMaterialLibrary)
  (Rendering|Material|SetVectorParameterValue "/Game/Environments/World/Backgrounds/Universal/MasterMaterials/MPC_Skybox_World.MPC_Skybox_World" "Morn_R=LitLength_G=LitRot_B=LitOffset" (Math|Color|MakeColor (Variables|SkyboxFogSettings|GetSkyboxFogMorningLitLength) (+ 4000000 (Math|Float|Clamp(Float) (Variables|SkyboxFogSettings|GetSkyboxFogMorningLitRotation) -4000000 4000000)) (+ 4000000 (Math|Float|Clamp(Float) (Variables|SkyboxFogSettings|GetSkyboxFogMorningLitOffset) -4000000 4000000))) :!class /Script/Engine.KismetMaterialLibrary)
  (Rendering|Material|SetVectorParameterValue "/Game/Environments/World/Backgrounds/Universal/MasterMaterials/MPC_Skybox_World.MPC_Skybox_World" "DayColor" (Variables|SkyboxFogSettings|GetSkyboxFogDayColor) :!class /Script/Engine.KismetMaterialLibrary)
  (Rendering|Material|SetVectorParameterValue "/Game/Environments/World/Backgrounds/Universal/MasterMaterials/MPC_Skybox_World.MPC_Skybox_World" "DayLitColor" (Variables|SkyboxFogSettings|GetSkyboxFogDayLitColor) :!class /Script/Engine.KismetMaterialLibrary)
  (Rendering|Material|SetVectorParameterValue "/Game/Environments/World/Backgrounds/Universal/MasterMaterials/MPC_Skybox_World.MPC_Skybox_World" "Day_R=Distance_G=DiffuseAmt_B=EmissiveAmt" (Math|Color|MakeColor (Variables|SkyboxFogSettings|GetSkyboxFogDayDistance) (Variables|SkyboxFogSettings|GetSkyboxFogDayDiffuseAmount) (Variables|SkyboxFogSettings|GetSkyboxFogDayEmissiveAmount)) :!class /Script/Engine.KismetMaterialLibrary)
  (Rendering|Material|SetVectorParameterValue "/Game/Environments/World/Backgrounds/Universal/MasterMaterials/MPC_Skybox_World.MPC_Skybox_World" "Day_R=LitLength_G=LitRot_B=LitOffset" (Math|Color|MakeColor (Variables|SkyboxFogSettings|GetSkyboxFogDayLitLength) (+ 4000000 (Math|Float|Clamp(Float) (Variables|SkyboxFogSettings|GetSkyboxFogDayLitRotation) -4000000 4000000)) (+ 4000000 (Math|Float|Clamp(Float) (Variables|SkyboxFogSettings|GetSkyboxFogDayLitOffset) -4000000 4000000))) :!class /Script/Engine.KismetMaterialLibrary)
  (Rendering|Material|SetVectorParameterValue "/Game/Environments/World/Backgrounds/Universal/MasterMaterials/MPC_Skybox_World.MPC_Skybox_World" "EveningColor" (Variables|SkyboxFogSettings|GetSkyboxFogEveningColor) :!class /Script/Engine.KismetMaterialLibrary)
  (Rendering|Material|SetVectorParameterValue "/Game/Environments/World/Backgrounds/Universal/MasterMaterials/MPC_Skybox_World.MPC_Skybox_World" "EveningLitColor" (Variables|SkyboxFogSettings|GetSkyboxFogEveningLitColor) :!class /Script/Engine.KismetMaterialLibrary)
  (Rendering|Material|SetVectorParameterValue "/Game/Environments/World/Backgrounds/Universal/MasterMaterials/MPC_Skybox_World.MPC_Skybox_World" "Eve_R=Distance_G=DiffuseAmt_B=EmissiveAmt" (Math|Color|MakeColor (Variables|SkyboxFogSettings|GetSkyboxFogEveningDistance) (Variables|SkyboxFogSettings|GetSkyboxFogEveningDiffuseAmount) (Variables|SkyboxFogSettings|GetSkyboxFogEveningEmissiveAmount)) :!class /Script/Engine.KismetMaterialLibrary)
  (Rendering|Material|SetVectorParameterValue "/Game/Environments/World/Backgrounds/Universal/MasterMaterials/MPC_Skybox_World.MPC_Skybox_World" "Eve_R=LitLength_G=LitRot_B=LitOffset" (Math|Color|MakeColor (Variables|SkyboxFogSettings|GetSkyboxFogEveningLitLength) (+ 4000000 (Math|Float|Clamp(Float) (Variables|SkyboxFogSettings|GetSkyboxFogEveningLitRotation) -4000000 4000000)) (+ 4000000 (Math|Float|Clamp(Float) (Variables|SkyboxFogSettings|GetSkyboxFogEveningLitOffset) -4000000 4000000))) :!class /Script/Engine.KismetMaterialLibrary)
  (Rendering|Material|SetVectorParameterValue "/Game/Environments/World/Backgrounds/Universal/MasterMaterials/MPC_Skybox_World.MPC_Skybox_World" "NightColor" (Variables|SkyboxFogSettings|GetSkyboxFogNightColor) :!class /Script/Engine.KismetMaterialLibrary)
  (Rendering|Material|SetVectorParameterValue "/Game/Environments/World/Backgrounds/Universal/MasterMaterials/MPC_Skybox_World.MPC_Skybox_World" "NightLitColor" (Variables|SkyboxFogSettings|GetSkyboxFogNightLitColor) :!class /Script/Engine.KismetMaterialLibrary)
  (Rendering|Material|SetVectorParameterValue "/Game/Environments/World/Backgrounds/Universal/MasterMaterials/MPC_Skybox_World.MPC_Skybox_World" "Night_R=Distance_G=DiffuseAmt_B=EmissiveAmt" (Math|Color|MakeColor (Variables|SkyboxFogSettings|GetSkyboxFogNightDistance) (Variables|SkyboxFogSettings|GetSkyboxFogNightDiffuseAmount) (Variables|SkyboxFogSettings|GetSkyboxFogNightEmissiveAmount)) :!class /Script/Engine.KismetMaterialLibrary)
  (Rendering|Material|SetVectorParameterValue "/Game/Environments/World/Backgrounds/Universal/MasterMaterials/MPC_Skybox_World.MPC_Skybox_World" "Night_R=LitLength_G=LitRot_B=LitOffset" (Math|Color|MakeColor (Variables|SkyboxFogSettings|GetSkyboxFogNightLitLength) (+ 4000000 (Math|Float|Clamp(Float) (Variables|SkyboxFogSettings|GetSkyboxFogNightLitRotation) -4000000 4000000)) (+ 4000000 (Math|Float|Clamp(Float) (Variables|SkyboxFogSettings|GetSkyboxFogNightLitOffset) -4000000 4000000))) :!class /Script/Engine.KismetMaterialLibrary))
